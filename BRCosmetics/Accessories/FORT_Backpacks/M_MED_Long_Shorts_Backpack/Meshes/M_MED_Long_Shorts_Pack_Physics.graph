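# Physics asset for the M_MED_Long_Shorts backpack cosmetic, giving its dangling attachments physical secondary motion.
# Small capsule bodies cover the pack, antenna, two grenades and walkie cord bones.
# Each attachment is constrained directly to the pack body in a star layout, with its own angular limits.
# The antenna has the tightest twist limit and the walkie cord has the loosest swing range.
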PhysicsAsset /BRCosmetics/Accessories/FORT_Backpacks/M_MED_Long_Shorts_Backpack/Meshes/M_MED_Long_Shorts_Pack_Physics
bodies {
 dyn_antenna [{shapeType=Capsule, length=2.5, radius=1.5, center=<1.7837 0 -0.0001>, rotation=<rot 90 0 0>}]
 dyn_grenade_a [{shapeType=Capsule, length=7.25, radius=3.5, center=<0 0 -3>}]
 dyn_grenade_b [{shapeType=Capsule, length=7.25, radius=3.5, center=<0 0 -3>}]
 dyn_walkie_cord [{shapeType=Capsule, length=2.75, radius=3, center=<4.1558 0 0>, rotation=<rot 90 0 0>}]
 pack [{shapeType=Capsule, length=1, radius=1}]
}
constraints [
 {bone1Name=dyn_grenade_b, bone2Name=pack, swing1LimitDegrees=20, swing2LimitDegrees=5, twistLimitDegrees=20}
 {bone1Name=dyn_grenade_a, bone2Name=pack, swing1LimitDegrees=25, swing2LimitDegrees=10, twistLimitDegrees=15}
 {bone1Name=dyn_antenna, bone2Name=pack, swing1LimitDegrees=15, swing2LimitDegrees=15, twistLimitDegrees=5}
 {bone1Name=dyn_walkie_cord, bone2Name=pack, swing1LimitDegrees=30, swing2LimitDegrees=30, twistLimitDegrees=25}
]
